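# M_Slowfield_Flare_01 is an unlit additive particle sprite material for a flare in the Slow Field gadget effect.
# A sphere mask on the UVs creates a centered, hard-edged disc procedurally, without textures.
# The disc multiplies particle color into the MF_FX_EmissiveAdjustments function for emissive.
# It also feeds a DepthFade with an adjustable FadeDistance parameter to soften intersections with scene geometry.
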Material /Game/Gadgets/Assets/Slow_Field/Materials/M_Slowfield_Flare_01
graph {
 properties {bUsedWithParticleSprites=true, blendMode=BLEND_Additive, pixelDepthOffsetMode=PDOM_Legacy, shadingModel=MSM_Unlit, usedShadingModels=Unlit}
 nodes [
  n1 = Constant() {r=0.5}
  n2 = SphereMask(A=TextureCoordinate(), B=@n1, Radius=@n1, Hardness=Constant() {r=1})
  DepthFade(Opacity=@n2, FadeDistance=ScalarParameter() {defaultValue=128, parameterName=FadeDistance})
  MaterialFunctionCall(Emissive=Multiply(A=ParticleColor().RGB, B=@n2)) {materialFunction=&MF_FX_EmissiveAdjustments}
 ]
}
